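# Unlit translucent material, depth test disabled, that draws winter snowstorm haze on a sphere that can follow the camera.
# Three cloud-texture layers scroll at weather-index-scaled speeds with vertex-interpolated UVs, faded by local height and depth fade.
# Low quality, forward and mobile paths keep only Layer 1.
# Weather event intensity drives opacity, time-of-day scalars tint emissive strength, and a Blueprint parameter blends camera position into the local-position offset.
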Material /Game/TimeOfDay/TODM/BR/s11/Weather/s11_30_Winter/M_SnowStormCameraSphere
graph {
 properties {bDisableDepthTest=true, blendMode=BLEND_Translucent, pixelDepthOffsetMode=PDOM_Legacy, shadingModel=MSM_Unlit, usedShadingModels=Unlit}
 nodes [
  n1 = Add(A=ScalarParameter() {parameterName=WeatherIndexFromBP})
  n2 = Multiply(A=Time(), B=Multiply(A=ScalarParameter() {defaultValue=0.02, parameterName=Layer3_ScollingSpeed}, B=Divide(A=@n1) {constB=3})) {constB=0.01}
  n3 = Frac(None=@n2)
  n4 = TextureCoordinate()
  n5 = Multiply(A=ComponentMask(None=@n4) {r=true}, B=ScalarParameter() {defaultValue=6, parameterName=Layer3_UTiling})
  n6 = Subtract(A=ComponentMask(None=@n4) {g=true}) {constB=0.66}
  n7 = ScalarParameter() {defaultValue=6, parameterName=Layer3_VTiling}
  n8 = CameraPositionWS()
  n9 = Add(A=Multiply(A=Multiply(A=@n6, B=@n7)) {constB=-1}, B=Multiply(A=ComponentMask(None=@n8) {b=true}) {constB=0.0001})
  n10 = Frac(None=Multiply(A=@n2) {constB=-0.5})
  n11 = AppendVector(A=Add(A=Add(A=@n3, B=@n5)), B=Add(B=Subtract(A=@n9, B=@n10)))
  n12 = OneMinus(None=TextureSample(UVs=VertexInterpolator(VS=@n11).PS) {texture=/Game/Athena/SafeZone/S11/T_TilingClouds_D}.R)
  n13 = ScalarParameter() {defaultValue=0.15, parameterName=Layer3_Opacity}
  n14 = CollectionParameter() {
   collection=&FortniteMaterialParameters
   parameterName=CurrentWeatherEventIntensity
  }
  n15 = Multiply(A=@n14, B=@n14)
  n16 = LinearInterpolate(A=@n13, B=Multiply(A=@n13) {constB=0.5}, Alpha=@n15)
  n17 = Saturate(None=Add(A=Multiply(A=@n12, B=@n16) {constB=0.4}, B=ScalarParameter() {defaultValue=0.05, parameterName=Layer3_OpacityAdd}) {constB=0.25})
  n18 = MaterialFunctionCall() {materialFunction=&LocalPosition}
  n19 = ComponentMask(None=@n18."Local Position (Excluding Offsets)") {b=true}
  n20 = Abs(None=Add(A=@n19) {constB=10})
  n21 = MaterialFunctionCall(X=Multiply(A=@n20) {constB=0.03}, Y=Multiply(A=@n20) {constB=0.02}, Z=Multiply(A=@n20) {constB=0.015}) {materialFunction=&MakeFloat3}
  n22 = MaterialFunctionCall(Float3=OneMinus(None=Power(Base=@n21.Result))) {materialFunction=&BreakOutFloat3Components}
  n23 = DepthFade(FadeDistance=ScalarParameter() {defaultValue=100000, parameterName=Global_DepthFadeDistance})
  n24 = Multiply(A=ScalarParameter() {defaultValue=1, parameterName=Layer3_DepthFadeDistanceScale}, B=@n23)
  n25 = Add(A=ScalarParameter() {parameterName=WeatherIndexFromBP})
  n26 = Multiply(A=Time(), B=Multiply(A=ScalarParameter() {defaultValue=0.075, parameterName=Layer2_ScollingSpeed}, B=Divide(A=@n25) {constB=3})) {constB=0.1}
  n27 = TextureCoordinate()
  n28 = ComponentMask(None=@n27) {r=true}
  n29 = ScalarParameter() {defaultValue=4, parameterName=Layer2_UTiling}
  n30 = ComponentMask(None=@n27) {g=true}
  n31 = Multiply(A=Multiply(A=Subtract(A=@n30) {constB=0.33}, B=ScalarParameter() {defaultValue=4, parameterName=Layer2_VTiling})) {constB=-1}
  n32 = Multiply(A=ComponentMask(None=CameraPositionWS()) {b=true}) {constB=0}
  n33 = Multiply(A=@n26) {constB=-0.25}
  n34 = AppendVector(A=Add(A=Frac(None=@n26), B=Multiply(A=@n28, B=@n29)), B=Subtract(A=Add(A=@n31, B=@n32), B=Frac(None=@n33)))
  n35 = ScalarParameter() {defaultValue=0.25, parameterName=Layer2_Opacity}
  n36 = Multiply(
   A=TextureSample(UVs=VertexInterpolator(VS=@n34).PS) {texture=&T_TilingClouds_D}.R
   B=LinearInterpolate(A=@n35, B=Multiply(A=@n35) {constB=0.5}, Alpha=@n15)
  ) {constB=0.35}
  n37 = ScalarParameter() {defaultValue=0.05, parameterName=Layer2_OpacityAdd}
  n38 = Multiply(A=Saturate(None=Add(A=@n36, B=@n37) {constB=0.25}), B=@n22.G)
  n39 = Saturate(None=Multiply(A=ScalarParameter() {defaultValue=5, parameterName=Layer2_DepthFadeDistanceScale}, B=@n23))
  n40 = ScalarParameter() {defaultValue=10, parameterName=Layer1_DepthFadeDistanceScale}
  n41 = Add(A=ScalarParameter() {parameterName=WeatherIndexFromBP})
  n42 = Multiply(A=Time(), B=Multiply(A=ScalarParameter() {defaultValue=0.25, parameterName=Layer1_ScollingSpeed}, B=Divide(A=@n41) {constB=3})) {constB=0.3}
  n43 = TextureCoordinate()
  n44 = ComponentMask(None=@n43) {r=true}
  n45 = ScalarParameter() {defaultValue=3, parameterName=Layer1_UTiling}
  n46 = Multiply(A=Multiply(A=ComponentMask(None=@n43) {g=true}, B=ScalarParameter() {defaultValue=3, parameterName=Layer1_VTiling})) {constB=-1}
  n47 = Multiply(A=ComponentMask(None=CameraPositionWS()) {b=true}) {constB=0.0002}
  n48 = Multiply(A=@n42) {constB=-0.2}
  n49 = AppendVector(A=Add(A=Frac(None=@n42), B=Multiply(A=@n44, B=@n45)), B=Subtract(A=Add(A=@n46, B=@n47), B=Frac(None=@n48)))
  n50 = OneMinus(None=TextureSample(UVs=VertexInterpolator(VS=@n49).PS) {texture=&T_TilingClouds_D}.R)
  n51 = Add(
   A=Multiply(A=Power(Base=@n50), B=ScalarParameter() {defaultValue=0.25, parameterName=Layer1_Opacity}) {constB=0.25}
   B=ScalarParameter() {parameterName=Layer1_OpacityAdd}
  ) {constB=0.25}
  n52 = Multiply(A=Saturate(None=Multiply(A=@n23, B=@n40)), B=Multiply(A=Saturate(None=@n51), B=@n22.R))
  n53 = Add(A=Multiply(A=Multiply(A=@n17, B=@n22.B), B=Saturate(None=@n24)), B=Add(A=Multiply(A=@n38, B=@n39), B=@n52))
  n54 = QualitySwitch(Default=@n53, Low=@n52, High=@n53, Medium=@n53)
  n55 = OneMinus(None=Saturate(None=Multiply(A=@n19) {constB=0.05}))
  n56 = Multiply(A=Abs(None=@n19)) {constB=0.015}
  n57 = Multiply(
   A=Add(A=ShadingPathSwitch(Default=@n54, Deferred=@n54, Forward=@n52, Mobile=@n52), B=Multiply(A=@n55) {constB=0.1})
   B=OneMinus(None=Power(Base=@n56))
  )
  n58 = CollectionParameter() {
   collection=&FortniteMaterialParameters
   parameterName=CurrentWeatherEventIntensity
  }
  n59 = Multiply(
   A=Multiply(A=Saturate(None=@n57), B=ScalarParameter() {defaultValue=1, parameterName=Global_OpacityScale})
   B=Power(Base=@n58) {constExponent=0.25}
  ) {constB=0.2}
  Multiply(A=@n59, B=VertexColor().A)
  n60 = ScalarParameter() {defaultValue=1, parameterName=TOD_EmissiveScale_Morning}
  n61 = ScalarParameter() {defaultValue=1, parameterName=TOD_EmissiveScale_Day}
  n62 = ScalarParameter() {defaultValue=1, parameterName=TOD_EmissiveScale_Evening}
  n63 = ScalarParameter() {defaultValue=1, parameterName=TOD_EmissiveScale_Night}
  Multiply(
   A=Multiply(
    A=MaterialFunctionCall(Value_Morning=@n60, Value_Day=@n61, Value_Evening=@n62, Value_Night=@n63) {materialFunction=&Scalar_Time_of_Day}.Result
    B=@n59
   )
   B=VectorParameter() {parameterName=EmissiveColor, defaultValue=<rgba 1 1 1 1>}.RGB
  )
  n64 = Constant3Vector()
  n65 = CameraPositionWS()
  n66 = ScalarParameter() {parameterName=ActivateCameraPos_FromBP}
  Reroute(None=Add(A=@n18.Local Position, B=LinearInterpolate(A=@n64, B=@n65, Alpha=@n66)))
  Saturate(None=Multiply(A=@n23) {constB=0.5})
 ]
}
